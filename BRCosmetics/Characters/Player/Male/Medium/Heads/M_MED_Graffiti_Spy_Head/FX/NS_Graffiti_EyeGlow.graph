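# Niagara effect for the Graffiti Spy character head that adds two glowing pink eyes and dark smoke over the face.
# Flare burst-spawns one persistent particle per eye, using spawn groups for mirrored offsets, each drawn with sprite, light and mesh renderers.
# Face_Smoke continuously emits SubUV sprites from a small hemisphere that sink under gravity, fade in and out, and shift from pink tint to near black through color curves.
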NiagaraSystem /BRCosmetics/Characters/Player/Male/Medium/Heads/M_MED_Graffiti_Spy_Head/FX/NS_Graffiti_EyeGlow
dependencies {
 usedDataInterfaces [&NiagaraDataInterfaceVectorCurve, &NiagaraDataInterfaceCurve]
 usedDynamicInputs [&ConvertVectorToPosition]
 usedModules [&SystemState]
}
emitters {
 Face_Smoke {
  emitterSpawnScript [/SetVariables(Emitter.SmokeDepthFade=5) {isSetParameters=true}]
  emitterUpdateScript [
   &EmitterState(Life Cycle Mode=::NewEnumerator1, Loop Duration=2, UseLoopDelay=false)
   &SpawnRate(SpawnRate=10, Use Spawn Probability=false)
  ]
  particleSpawnScript [
   &InitializeParticle(
    Color=<rgba 0.0025 0.0025 0.0025 3.5>
    Lifetime Max=1.5
    Lifetime Mode=::NewEnumerator1
    Mass Mode=::NewEnumerator3
    Sprite Rotation Mode=::NewEnumerator2
    Sprite Size Mode=::NewEnumerator1
    Uniform Sprite Size Max=20
    Uniform Sprite Size Min=12
    UsePositionOffset=false
    Write Material Random=true
   )
   &ShapeLocation(
    Hemisphere Distribution=<0.5 1>
    MaskSpawnGroup=false
    Offset=<0 4 0>
    Offset Mode=::NewEnumerator0
    Override Seed=false
    Sphere Radius=4
    Sphere Surface Distribution=1
    UseOwnerScale=true
   )
   &AddVelocity(Module.Use Falloff Distance=false, Origin Offset=<0 2 0>, Override Seed=false, Velocity Mode=::NewEnumerator1, Velocity Speed=5) {module=AddVelocity001}
   &CameraOffset(Camera Offset Amount=2)
   &DynamicMaterialParameters(
    Index 0 Param 1=1
    Module.Param0WriteEnabled=true
    Module.Param1WriteEnabled=false
    Module.Param2WriteEnabled=false
    Module.Param3WriteEnabled=false
   )
  ]
  particleUpdateScript [
   &ParticleState()
   &GravityForce(Gravity=<0 0 -15>)
   &Drag(Drag=0.25, Use Linear Drag=true, Use Rotational Drag=false)
   &ScaleColor(
    Scale Alpha=&FloatFromCurve(FloatCurve=DataInterface() {propertyValues {Curve {keys [
         {interpMode=RCIM_Cubic, tangentMode=RCTM_User, time=0, value=0}
         {interpMode=RCIM_Cubic, tangentMode=RCTM_User, time=0.5, value=1}
         {tangentMode=RCTM_User, time=1, value=0}
        ]}}})
    Scale RGB=&VectorFromCurve(VectorCurve=DataInterface() {propertyValues {
       XCurve {keys [
         {arriveTangent=-31.0911, interpMode=RCIM_Cubic, leaveTangent=-31.0911, tangentMode=RCTM_User, time=0, value=15}
         {interpMode=RCIM_Cubic, tangentMode=RCTM_User, time=0.7, value=1}
        ]}
       YCurve {keys [<at 0 1>,<at 0.7 1>]}
       ZCurve {keys [{arriveTangent=-19.7928, interpMode=RCIM_Cubic, leaveTangent=-19.7928, tangentMode=RCTM_User, time=0, value=8}, <at 0.7 1>]}
      }})
    ScaleA=true
    ScaleRGB=true
   )
   &SolveForcesAndVelocity(Clamp Velocity=false, Limit Acceleration=false)
   &SubUVAnimation(End Frame=48, Start Frame=5)
   &ScaleSpriteSize(Uniform Curve Sprite Scale=DataInterface() {propertyValues {Curve {keys [
        {arriveTangent=0.5974, interpMode=RCIM_Cubic, leaveTangent=0.5974, tangentMode=RCTM_User, time=0, value=0.7}
        {arriveTangent=0.1871, interpMode=RCIM_Cubic, leaveTangent=0.1871, tangentMode=RCTM_User, time=1, value=1}
       ]}}})
  ]
  rendererClasses [&NiagaraSpriteRendererProperties]
  renderers [{rendererClass=&NiagaraSpriteRendererProperties}]
 }
 Flare {
  emitterSpawnScript [/SetVariables(Emitter.GlowDepthFade=2) {isSetParameters=true}]
  emitterUpdateScript [
   &EmitterState(
    Inactive Response=::NewEnumerator1
    Life Cycle Mode=::NewEnumerator1
    Loop Behavior=::NewEnumerator1
    Loop Duration Mode=::NewEnumerator1
    UseLoopDelay=false
   )
   &SpawnBurst_Instantaneous(Spawn Count=1, Use Spawn Probability=false)
   &SpawnBurst_Instantaneous(Spawn Count=1, Spawn Group=1, Use Spawn Probability=false) {module=SpawnBurst_Instantaneous001}
  ]
  particleSpawnScript [
   /Niagara/Modules/Spawn/Initialization/InitializeParticle(
    Mesh Scale=<1.3 1.3 1.3>
    Module.Write Color=true
    Module.Write Lifetime=true
    Module.Write Mass=true
    Module.Write Material Random=true
    Module.Write Position=false
    Module.Write Scale=false
    Module.Write SpriteRotation=false
    Module.Write SpriteSize=true
    Position=&SimulationPosition()
    Sprite Size=<20 10>
   )
   &SystemLocation(MaskSpawnGroup=false, Offset=<4.5 8.5 2.75>)
   &SystemLocation(MaskSpawnGroup=true, Offset=<4.5 8.5 -2.75>, Spawn Group Mask=1) {module=SystemLocation001}
   /SetVariables(Particles.EyeGlow_Position:Vector3f=&Add_Vector(A=Linked() {linkedVariable {name=Particles.Position, type=NiagaraPosition}}, B=<2.7 -2 0>)) {enabled=false, isSetParameters=true}
   /SetVariables(Particles.LightPosition:NiagaraPosition=&Add_Vector(A=Linked() {linkedVariable {name=Particles.Position, type=NiagaraPosition}}, B=<-1.5 1 0>)) {isSetParameters=true}
  ]
  particleUpdateScript [
   &ParticleState(Kill Particles When Lifetime Has Elapsed=false, Let Infinitely Lived Particles Die When Emitter Deactivates=true)
   /SetVariables(Particles.EyeGlow_Color:LinearColor=<rgba 0.5 0.0499 0.1373 1>) {isSetParameters=true}
   /SetVariables(Particles.LightGlow:LinearColor=<rgba 5 0.4995 1.3734 1>) {isSetParameters=true}
   &CameraOffset(Camera Offset Amount=1)
   &SolveForcesAndVelocity(Clamp Velocity=false, Limit Acceleration=false)
   &FN_Light_Attributes(Create Exponent Attribute=true, Create Radius Attribute=true, Radius=300)
  ]
  rendererClasses [
   &NiagaraSpriteRendererProperties
   &NiagaraLightRendererProperties
   &NiagaraMeshRendererProperties
  ]
  renderers [
   {rendererClass=&NiagaraSpriteRendererProperties}
   {rendererClass=&NiagaraLightRendererProperties, rendererIndex=1}
   {rendererClass=&NiagaraMeshRendererProperties, rendererIndex=2}
  ]
 }
}
summary {
 systemName=NS_Graffiti_EyeGlow
 emitters [
  {
   emitterName=Flare
   rendererClasses [
    &NiagaraSpriteRendererProperties
    &NiagaraLightRendererProperties
    &NiagaraMeshRendererProperties
   ]
  }
  {emitterName=Face_Smoke, rendererClasses [&NiagaraSpriteRendererProperties]}
 ]
 userVariables [
  User.EyeGlowMultiplier {type=NiagaraFloat, defaultValue=0, description=""}
  User.Right2ImpactVelocity {type=Vector3f, description="", defaultValue {struct=Vector3f, value=<0 0 0>}}
 ]
}
